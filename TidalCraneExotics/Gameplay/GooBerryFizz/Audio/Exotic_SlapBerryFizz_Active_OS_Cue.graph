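# One-shot sound cue for the activation of the Slap Berry Fizz exotic, in the Gadgets Items sound class.
# A Random node picks between two equally weighted gravity sphere enter waves reused from Fort_Gadgets, then feeds a Modulator at the root.
# It uses the shared ability status-effect activation attenuation at reduced volume.
# A concurrency override allows one instance at a time, blocking new plays and setting a short retrigger time.
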SoundCue /TidalCraneExotics/Gameplay/GooBerryFizz/Audio/Exotic_SlapBerryFizz_Active_OS_Cue
root=Modulator_0
nodes [
 {id=Modulator_0, children [Random_1]}
 {id=Random_1, children [WavePlayer_10, WavePlayer_11], properties {weights [1,1]}}
 {id=WavePlayer_10, properties {soundWaveAssetPtr=/FNGameplayCues/Sounds/Fort_Gadgets/GravitySphere/fort_gravity_sphere_enternew_01}}
 {id=WavePlayer_11, properties {soundWaveAssetPtr=/FNGameplayCues/Sounds/Fort_Gadgets/GravitySphere/fort_gravity_sphere_enternew_02}}
]
properties {
 attenuationSettings=/Game/Sounds/Attenuation/Fort_Ability_Status_Effect_Activation_Att
 bOverrideConcurrency=true
 maxDistance=6000
 soundClassObject=&Fort_Gadgets_Items
 volumeMultiplier=0.4
 concurrencyOverrides {maxCount=1, resolutionRule=PreventNew, retriggerTime=0.3}
}
