Material /TidalCraneExotics/Gameplay/GooBerryFizz/FX/MD_TidalCrane_TeamSpray_AOE
graph {
 properties {
  bAutomaticallySetUsageInEditor=false
  bCastRayTracedShadows=false
  bTangentSpaceNormal=false
  blendMode=BLEND_Translucent
  materialDomain=MD_DeferredDecal
  pixelDepthOffsetMode=PDOM_Legacy
  shadingModel=MSM_Unlit
  usedShadingModels=Unlit
 }
 nodes [
  Constant()
  n1 = ScalarParameter() {defaultValue=1, parameterName=Specular - Water}
  LinearInterpolate(A=@n1, B=ScalarParameter() {defaultValue=1, parameterName=Specular - Foam}, Alpha=NamedRerouteUsage()) {constA=1, constB=0.5}
  n2 = ScalarParameter() {defaultValue=0.1, parameterName=Roughness - Water}
  LinearInterpolate(A=@n2, B=ScalarParameter() {parameterName=Roughness - Foam}, Alpha=NamedRerouteUsage()) {constA=0.2, constB=0.8}
  MaterialFunctionCall() {materialFunction=/Game/Weapons/Prototype/SprayGun/Materials/MF_DecalFacingMask}
  MaterialFunctionCall(Emissive=Multiply(A=NamedRerouteUsage(), B=ScalarParameter() {defaultValue=0.25, parameterName=Emissive Multiplier})) {materialFunction=&MF_FX_EmissiveAdjustments}
  n3 = Abs(None=Multiply(A=NamedRerouteUsage(), B=ScalarParameter() {defaultValue=1, parameterName=Droplets Tiling Multiplier}))
  n4 = Add(
   A=NamedRerouteUsage()
   B=TextureSample(UVs=@n3) {samplerType=SAMPLERTYPE_LinearColor, texture=/Game/Global/Textures/Default/Noise/T_Wrap_SnowStreak_Noise}.G
  )
  n5 = Constant() {r=1}
  n6 = ScalarParameter() {defaultValue=6, parameterName=Droplet Expansion Power}
  n7 = Power(Base=OneMinus(None=Fmod(A=@n4, B=@n5)), Exp=@n6)
  n8 = NamedRerouteUsage()
  n9 = Add(A=Fmod(A=Multiply(A=@n3) {constB=32}, B=Constant() {r=1}), B=LinearInterpolate(A=NamedRerouteUsage(), Alpha=@n8) {constB=0})
  n10 = Constant() {r=0.5}
  n11 = ScalarParameter() {defaultValue=0.05, parameterName=Droplet Radius Min}
  n12 = LinearInterpolate(B=@n11, Alpha=@n8) {constA=0.5, constB=0.05}
  n13 = Constant()
  n14 = Add(A=@n3) {constB=0.5156}
  n15 = Add(
   A=NamedRerouteUsage()
   B=TextureSample(UVs=@n14) {samplerType=SAMPLERTYPE_LinearColor, texture=&T_Wrap_SnowStreak_Noise}.G
  )
  n16 = Constant() {r=1}
  n17 = Power(Base=OneMinus(None=Fmod(A=@n15, B=@n16)), Exp=@n6)
  n18 = NamedRerouteUsage()
  n19 = Add(A=Fmod(A=Multiply(A=@n14) {constB=32}, B=Constant() {r=1}), B=LinearInterpolate(A=NamedRerouteUsage(), Alpha=@n18) {constB=0})
  n20 = Constant() {r=0.5}
  n21 = LinearInterpolate(B=@n11, Alpha=@n18) {constA=0.5, constB=0.02}
  n22 = Constant()
  n23 = Add(
   A=Multiply(
    A=NamedRerouteDeclaration(None=@n7) {name=Drop Noise Mult A, nodeColor=<rgba 0.5234 0 1 1>}
    B=SphereMask(A=@n9, B=@n10, Radius=@n12, Hardness=@n13)
   )
   B=Multiply(
    A=NamedRerouteDeclaration(None=@n17) {name=Drop Noise Mult B, nodeColor=<rgba 0.5234 0 1 1>}
    B=SphereMask(A=@n19, B=@n20, Radius=@n21, Hardness=@n22)
   )
  )
  Multiply(A=@n23, B=Abs(None=Sine(None=@n23)))
  Multiply(A=TextureCoordinate())
  NamedRerouteDeclaration(None=Saturate(None=Constant())) {name=Foam Mask, nodeColor=<rgba 1 0 0.8906 1>}
  n24 = VectorParameter() {parameterName=Base Color C, defaultValue=<rgba 0.25 0 1 0.15>}
  n25 = VectorParameter() {parameterName=Base Color B, defaultValue=<rgba 0.25 0 1 0.15>}
  n26 = TextureSampleParameter2D(UVs=Rotator() {speed=1}) {parameterName=Rotating Color Mask, texture=/Game/Effects/Fort_Effects/Textures/Tile/T_Smooth_Tile_Chromatic}
  n27 = Saturate(None=Multiply(A=Power(Base=@n26.R) {constExponent=1}) {constB=2})
  n28 = VectorParameter() {parameterName=Base Color A, defaultValue=<rgba 0 1 4 1>}
  n29 = Power(Base=Saturate(None=NamedRerouteUsage()), Exp=ScalarParameter() {defaultValue=3, parameterName=Droplet Color Lerp Power}) {constExponent=6}
  NamedRerouteDeclaration(None=LinearInterpolate(A=LinearInterpolate(A=@n24.RGB, B=@n25.RGB, Alpha=@n27), B=@n28.RGB, Alpha=@n29)) {name=Diffuse, nodeColor=<rgba 0 0.5078 1 1>}
  NamedRerouteDeclaration(None=@n23) {name=Droplets Mask, nodeColor=<rgba 0.5938 0 1 1>}
  NamedRerouteDeclaration(None=Multiply(A=Time(), B=ScalarParameter() {defaultValue=2, parameterName=Droplet Time Multiplier})) {name=Droplet Time, nodeColor=<rgba 0 0.6016 1 1>}
  n30 = TextureSample(UVs=Multiply(A=NamedRerouteUsage(), B=ScalarParameter() {defaultValue=24, parameterName=Droplet UV Distortion Tile})) {samplerType=SAMPLERTYPE_Normal, texture=/Game/Global/Textures/Default/Noise/T_SwirlNoise_N}
  n31 = ScalarParameter() {defaultValue=0.2, parameterName=Droplet UV Distortion Amount}
  NamedRerouteDeclaration(None=ComponentMask(None=Multiply(A=@n30.RGB, B=@n31) {constB=0.01}) {g=true, r=true}) {name=Droplet UV Distortion, nodeColor=<rgba 0 1 0.7891 1>}
  n32 = TextureSampleParameter2D(UVs=Multiply(A=NamedRerouteUsage(), B=ScalarParameter() {defaultValue=0.5, parameterName=Water Tile})) {parameterName=Water Mask, texture=/Game/Global/Textures/Default/Caustics/T_Fractal_01}
  n33 = Add(A=Multiply(A=NamedRerouteUsage(), B=ScalarParameter() {defaultValue=0.5, parameterName=Water Tile})) {constB=0.6}
  NamedRerouteDeclaration(None=LinearInterpolate(
    A=ConstantBiasScale(None=@n32.R) {bias=0.1, scale=0.9091}
    Alpha=TextureSampleParameter2D(UVs=@n33) {parameterName=Water Mask, texture=&T_Fractal_01}.R
   )) {name=Soft Noise Mask, nodeColor=<rgba 0.8516 0 1 1>}
  NamedRerouteDeclaration(None=Multiply(A=LinearInterpolate(A=@n25.A, B=@n28.A, Alpha=@n29), B=NamedRerouteUsage())) {name=Opacity, nodeColor=<rgba 0 0.2969 1 1>}
  NamedRerouteDeclaration(None=@n1) {name=Specular, nodeColor=<rgba 0.3594 1 0 1>}
  NamedRerouteDeclaration(None=@n2) {name=Roughness, nodeColor=<rgba 0.0312 1 0 1>}
  n34 = NamedRerouteUsage()
  n35 = ScalarParameter() {defaultValue=1, parameterName=Droplet Mask Power}
  n36 = Multiply(
   A=Add(A=Power(Base=@n34, Exp=@n35) {constExponent=1.5}, B=ScalarParameter() {defaultValue=0.1, parameterName=Droplet Mask Add}) {constB=0.2}
   B=Multiply(A=NamedRerouteUsage(), B=NamedRerouteUsage())
  ) {constA=1}
  n37 = ScalarParameter() {defaultValue=0.5, parameterName=Opacity Softness}
  NamedRerouteDeclaration(None=Saturate(None=Divide(A=@n36, B=@n37) {constB=0.25})) {name=Water Mask, nodeColor=<rgba 0.0547 0 1 1>}
  NamedRerouteDeclaration(None=DecalColor().A) {name=Decal Opacity, nodeColor=<rgba 1 0 0.4922 1>}
  n38 = Constant3Vector() {constant=<rgba 0 0 1 0>}
  NamedRerouteDeclaration(None=@n38) {name=Normal, nodeColor=<rgba 1 0.8438 0 1>}
  NamedRerouteDeclaration(None=Divide(
    A=WorldPosition() {worldPositionShaderOffset=WPT_ExcludeAllShaderOffsets}.XY
    B=ScalarParameter() {defaultValue=2000, parameterName=World UV Scale}
   ) {constB=1000}) {name=World UV, nodeColor=<rgba 0.2422 0 1 1>}
  n39 = Constant() {r=0.5}
  n40 = SphereMask(A=LightVector(), B=@n39, Radius=@n39, Hardness=ScalarParameter() {parameterName=Radius Hardness})
  n41 = Multiply(A=Sine(None=Power(Base=@n40) {constExponent=1}) {period=2}, B=ScalarParameter() {defaultValue=2, parameterName=Radial Mask Multiplier}) {constB=3}
  n42 = Multiply(A=DotProduct(A=MaterialFunctionCall() {materialFunction=/Game/Weapons/Prototype/SprayGun/Materials/MF_DecalNormal}.WorldNormal, B=ObjectOrientation())) {constB=-1}
  MaterialExpressionSaturate_5 = Saturate(None=LinearInterpolate(Alpha=Clamp(None=@n42)) {constB=2})
  n43 = CameraVectorWS()
  n44 = ObjectOrientation()
  MaterialExpressionSaturate_2 = Saturate(None=Multiply(A=DotProduct(A=@n43, B=@n44)) {constB=-100})
  n45 = Custom!(Default=@MaterialExpressionSaturate_5, Android=@MaterialExpressionSaturate_2) {
   code="##if COMPILER_GLSL_ES2 || COMPILER_GLSL_ES3_1\r\n\treturn Android;\r\n##else\r\n\treturn Default;\r\n##endif"
   description=AndroidSwitch
   outputType=CMOT_Float1
   inputs [
    {inputName=Default, input {expression=&self:MaterialExpressionSaturate_5}}
    {inputName=Android, input {expression=&self:MaterialExpressionSaturate_2}}
   ]
  }
  NamedRerouteDeclaration(None=Multiply(A=Multiply(A=@n41, B=@n45), B=NamedRerouteUsage())) {name=Radial Mask, nodeColor=<rgba 0 0.9766 1 1>}
  NamedRerouteUsage()
  NamedRerouteUsage()
  NamedRerouteUsage()
  NamedRerouteUsage()
  NamedRerouteUsage()
  n46 = Multiply(A=Time(), B=ScalarParameter() {defaultValue=4, parameterName=Time Multiplier}) {constB=2}
  n47 = Constant() {r=1}
  n48 = Power(Base=Fmod(A=Add(A=@n46) {constB=0.5}, B=@n47)) {constExponent=0.5}
  Saturate(None=ConstantBiasScale(None=Cosine(None=@n48)))
  n49 = Power(Base=Fmod(A=@n46, B=@n47)) {constExponent=0.5}
  Saturate(None=ConstantBiasScale(None=Cosine(None=@n49)))
  TextureSampleParameter2D(UVs=Multiply(A=NamedRerouteUsage(), B=ScalarParameter() {defaultValue=2, parameterName=Foam Tile})) {parameterName=Foam Mask, samplerType=SAMPLERTYPE_LinearColor, texture=/Game/Global/Textures/Default/Noise/T_Noise01_Inverse}
  n50 = Multiply(A=NamedRerouteUsage(), B=ScalarParameter() {defaultValue=8, parameterName=Bubble Tile}) {constB=5}
  TextureSampleParameter2D(UVs=@n50) {parameterName=Bubble Mask, samplerType=SAMPLERTYPE_LinearGrayscale, texture=/Game/Global/Textures/Default/Noise/T_Bubble_B}
  n51 = NamedRerouteUsage()
  Transform(None=LinearInterpolate(
    A=@n38
    B=TextureSampleParameter2D(UVs=@n50) {parameterName=Foam, samplerType=SAMPLERTYPE_Normal, texture=/Game/Global/Textures/Default/Noise/T_Bubble_B_N}.RGB
    Alpha=Multiply(A=@n51) {constB=0.5}
   )) {transformSourceType=TRANSFORMSOURCE_World, transformType=TRANSFORM_Instance}
  VertexInterpolator(VS=PreSkinnedPosition())
 ]
}
