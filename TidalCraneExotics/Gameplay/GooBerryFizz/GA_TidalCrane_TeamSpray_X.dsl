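; Gameplay ability for the GooBerryFizz team spray, which gives the user and targeted pawns health or shields plus secondary and low-gravity effects.
; Each incremental event commits on the server, then runs target selection and applies effects to the avatar and each target.
; SelectHealthOrShieldGE picks heals below max health, otherwise shields, and ends the ability for downed pawns.
; A FlipFlop toggles the shield cue on alternate applications.
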(event EventApplyEffectIncremental
  (bind _returnvalue (Ability|GetAvatarActorfromActorInfo))
  (bind _shouldplayshieldfx (Variables|Default|GetShouldPlayShieldFX?))
  (if (Networking|IsServer)
    (bind _returnvalue_2 (Ability|CommitAbility))
    (if _returnvalue_2
      (Utilities|FlowControl|FlipFlop
        (:A
          (Variables|Default|SetShouldPlayShieldFX? true))
        (:B
          (Variables|Default|SetShouldPlayShieldFX? false)))
      (bind (_asynctaskproxy _targetdata) (Ability|Tasks|DoTargetSelection "None" "(TagName=\"Targeting.ApplicationTag.0\")" true)
        (:then
          (bind _ge_selected (CallFunction|SelectHealthorShieldGE (CallFunction|GetAvatarasFortPlayerPawn)))
          (Fortnite|ApplyGameplayEffectsToActor _returnvalue (Utilities|Array|MakeArray (Variables|Default|GetGESecondaryEffectToApply) (Variables|Default|GetGELowGrav) _ge_selected) (Transformation|GetActorLocation _returnvalue) false 0 (Ability|GetActivatingAbilityComponent))
          (if _shouldplayshieldfx
            (CallFunction|ApplyShieldsCue _returnvalue)))
        (:Targeted
          (for _array_element (Ability|TargetData|GetAllActorsfromTargetData _targetdata)
            (bind _asfort_pawn (Utilities|Casting|CastToFortPawn _array_element))
            (bind _ge_selected_2 (CallFunction|SelectHealthorShieldGE _asfort_pawn))
            (Fortnite|ApplyGameplayEffectsToActor _asfort_pawn (Utilities|Array|MakeArray _ge_selected_2 (Variables|Default|GetGESecondaryEffectToApply) (Variables|Default|GetGELowGrav)) (Transformation|GetActorLocation (Ability|GetAvatarActorfromActorInfo)) false 0 (Ability|GetActivatingAbilityComponent))
            (if _shouldplayshieldfx
              (CallFunction|ApplyShieldsCue _array_element)))))
      (else
        (Ability|EndAbility)))))

(fn SelectHealthOrShieldGE
  (params (TargetPawn :type Fort%20Pawn%20Object%20Reference))
  (if (Health|IsDBNO TargetPawn)
    (Ability|EndAbility)
    (elif (< (FortPawn|GetHealth TargetPawn) (FortPawn|GetMaxHealth TargetPawn))
      (return (Variables|Default|GetGE%20Heals))
      (else
        (return (Variables|Default|GetGE%20Shields))))))
